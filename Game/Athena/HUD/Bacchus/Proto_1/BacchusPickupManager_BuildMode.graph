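# Build-mode pickup manager HUD widget for the Bacchus prototype, a FortPickupManagerWidget subclass that displays building resources.
# An inherited horizontal panel holds large-icon PackResource widgets for wood, stone and metal, plus a gold currency entry that starts collapsed.
# A TutorialHighlight with a static square outline overlays the panel, extended past its edges by negative padding.
# The event graph is empty, so any behavior comes from the parent class and child widgets.
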WidgetBlueprint /Game/Athena/HUD/Bacchus/Proto_1/BacchusPickupManager_BuildMode
graphs [EventGraph {links [], nodes []}]
variables []
widgets {
 info {inheritedWidgetCount=2, parentClass=/Script/FortniteUI.FortPickupManagerWidget, rootWidgetClass=&Overlay, widgetCount=7}
 widgets [
  Overlay_0 {widgetClassPath=&Overlay, widget=Overlay_0}
  Panel_AllResources {widgetClassPath=&HorizontalBox, bInherited=true, parent=Overlay_0, widget=Panel_AllResources, slot {class=OverlaySlot}}
  "[Pack Resource] Wood" {
   widgetClassPath=/Game/UI/InGame/HUD/PackResource
   parent=Panel_AllResources
   widget="[Pack Resource] Wood"
   properties {bUseLargeIcon=true, resourceType=Wood, visibility=Visible}
   slot {class=HorizontalBoxSlot}
  }
  "[Pack Resource] Stone" {
   widgetClassPath=&PackResource
   parent=Panel_AllResources
   widget="[Pack Resource] Stone"
   properties {bUseLargeIcon=true, resourceType=Stone, visibility=Visible}
   slot {class=HorizontalBoxSlot}
  }
  "[Pack Resource] Metal" {
   widgetClassPath=&PackResource
   parent=Panel_AllResources
   widget="[Pack Resource] Metal"
   properties {bUseLargeIcon=true, visibility=Visible}
   slot {class=HorizontalBoxSlot}
  }
  PackResourceGoldCurrency {
   widgetClassPath=&PackResource
   parent=Panel_AllResources
   widget=PackResourceGoldCurrency
   properties {bUseLargeIcon=true, resourceType=GoldCurrency, visibility=Collapsed}
   slot {class=HorizontalBoxSlot}
  }
  TutorialHighlight {
   widgetClassPath=&TutorialHighlight
   parent=Overlay_0
   widget=TutorialHighlight
   properties {
    customHighlightWidget=/Game/Athena/HUD/Tutorial/Tutorial_Outline_Square_Static.Tutorial_Outline_Square_Static_C
    visibility=SelfHitTestInvisible
   }
   slot {class=OverlaySlot, properties {horizontalAlignment=HAlign_Fill, verticalAlignment=VAlign_Fill, padding=<ltrb -20 -80 -20 -20>}}
  }
 ]
}
