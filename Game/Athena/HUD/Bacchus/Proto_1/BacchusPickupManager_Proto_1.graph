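# Prototype Bacchus HUD pickup manager, based on FortPickupManagerWidget, showing carried resources under the top-left PickupManager element tag.
# A top row holds wood, stone, metal and collapsed gold counters, collapsed no-build placeholders and a separately wrapped global currency widget.
# A right-aligned second row holds a tag-filtered ingredient list and hidden reboot-card widget.
# Layout-tool open and close events toggle that reboot-card widget's mobile layout-editor visibility.
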WidgetBlueprint /Game/Athena/HUD/Bacchus/Proto_1/BacchusPickupManager_Proto_1
graphs [EventGraph {nodes [
   HandleOnOpenLayoutTool() {is=title K2Node_CustomEvent}
   > UpdateVisibilityForMobileLayoutEditor!(
    execute
    self:Reboot Cards Collected Widget Object Reference=Get RebootCardsCollectedWidget() -> (RebootCardsCollectedWidget:Reboot Cards Collected Widget Object Reference)
    bInLayoutMode:bool=true
   ) -> (then)
   HandleOnCloseLayoutTool() {is=title K2Node_CustomEvent}
   > UpdateVisibilityForMobileLayoutEditor(self=Get RebootCardsCollectedWidget(), bInLayoutMode=false)
  ]}]
variables []
widgets {
 info {
  inheritedWidgetCount=2
  namedSlotCount=2
  parentClass=/Script/FortniteUI.FortPickupManagerWidget
  rootWidgetClass=/Game/Athena/HUD/FortHUDElementWrapper.FortHUDElementWrapper_C
  widgetCount=16
 }
 widgets [
  FortHUDElementWrapper_61 {
   widgetClassPath=/Game/Athena/HUD/FortHUDElementWrapper
   bIsVariable=true
   widget=FortHUDElementWrapper_61
   properties {elementTag {tagName=HUD.TopLeft.PickupManager}}
  }
  Overlay_0 {widgetClassPath=&Overlay, namedSlotHost=FortHUDElementWrapper_61, widget=Overlay_0}
  VerticalBox_2 {widgetClassPath=&VerticalBox, parent=Overlay_0, widget=VerticalBox_2, slot {class=OverlaySlot}}
  Panel_AllResources {widgetClassPath=&HorizontalBox, bInherited=true, parent=VerticalBox_2, widget=Panel_AllResources, slot {class=VerticalBoxSlot}}
  "[Pack Resource] Wood" {
   widgetClassPath=/Game/UI/InGame/HUD/PackResource
   parent=Panel_AllResources
   widget="[Pack Resource] Wood"
   properties {resourceType=Wood}
   slot {class=HorizontalBoxSlot}
  }
  "[Pack Resource] Stone" {
   widgetClassPath=&PackResource
   parent=Panel_AllResources
   widget="[Pack Resource] Stone"
   properties {resourceType=Stone}
   slot {class=HorizontalBoxSlot}
  }
  "[Pack Resource] Metal" {widgetClassPath=&PackResource, parent=Panel_AllResources, widget="[Pack Resource] Metal", slot {class=HorizontalBoxSlot}}
  Spacer_NoBuild {
   widgetClassPath=&Spacer
   bIsVariable=true
   parent=Panel_AllResources
   widget=Spacer_NoBuild
   properties {visibility=Collapsed, size=<120 0>}
   slot {class=HorizontalBoxSlot}
  }
  PackResourceGoldCurrency {
   widgetClassPath=&PackResource
   parent=Panel_AllResources
   widget=PackResourceGoldCurrency
   properties {bDesignTime_IsResourceVisible=false, resourceType=GoldCurrency, visibility=Collapsed}
   slot {class=HorizontalBoxSlot}
  }
  SIzeBox_NoBuildIcon {
   widgetClassPath=&SizeBox
   bIsVariable=true
   parent=Panel_AllResources
   widget=SIzeBox_NoBuildIcon
   properties {bOverride_HeightOverride=true, bOverride_WidthOverride=true, heightOverride=72, visibility=Collapsed, widthOverride=64}
   slot {class=HorizontalBoxSlot, properties {horizontalAlignment=HAlign_Center, verticalAlignment=VAlign_Center}}
  }
  FortHUDElementWrapper_Cash {
   widgetClassPath=&FortHUDElementWrapper
   bIsVariable=true
   parent=Panel_AllResources
   widget=FortHUDElementWrapper_Cash
   properties {elementTag {tagName=HUD.CashResource}, hUDElementTag {gameplayTags [{tagName=HUD.Athena.HideForCreative}]}}
   slot {class=HorizontalBoxSlot}
  }
  GlobalCurrencyWidget {
   widgetClassPath=/Game/Athena/HUD/GlobalCurrencyWidget
   bIsVariable=true
   namedSlotHost=FortHUDElementWrapper_Cash
   widget=GlobalCurrencyWidget
  }
  HorizontalBox_0 {
   widgetClassPath=&HorizontalBox
   parent=VerticalBox_2
   widget=HorizontalBox_0
   slot {class=VerticalBoxSlot, properties {horizontalAlignment=HAlign_Right}}
  }
  IngredientList {
   widgetClassPath=/Game/Athena/HUD/AthenaHUDItemList
   parent=HorizontalBox_0
   widget=IngredientList
   properties {
    elementTag {tagName=HUD.Athena.IngredientsList}
    hUDElementTag {gameplayTags [
      {tagName=HUD.Athena.HideOnBus}
      {tagName=Athena.HideForCapture}
      {tagName=HUD.Athena.HideForGadget}
      {tagName=HUD.Athena.Hide.WithInventory}
      {tagName=HUD.Athena.Hide.WithRCPawn}
      {tagName=HUD.Athena.HideForAbility}
      {tagName=HUD.Event.HideForSpecialEvent}
      {tagName=Athena.Creative.HideForLoading}
      {tagName=HUD.Hide.WhileLoading}
      {tagName=HUD.Athena.HideForPostGame}
      {tagName=HUD.Athena.HideforRadial}
      {tagName=HUD.Athena.HideforNoBuilding}
      {tagName=HUD.Athena.HideforVehicle}
      {tagName=HUD.Athena.Hide.WithMap}
      {tagName=HUD.Athena.HideForConversation}
      {tagName=Athena.Creative.HUDController.HideCraftingResources}
     ]}
    itemFilter {bRequiresItemDetails=false, itemTypeFilters [WorldResource]}
    itemFilterTags {gameplayTags [{tagName=Item.Ingredient}]}
   }
   slot {class=HorizontalBoxSlot, properties {horizontalAlignment=HAlign_Right, padding=<ltrb 0 8 0 16>}}
  }
  RebootCardsCollectedWidget {
   widgetClassPath=/Game/Athena/HUD/RebootCardsCollectedWidget
   bIsVariable=true
   parent=HorizontalBox_0
   widget=RebootCardsCollectedWidget
   properties {noCardVisibility=Hidden, visibility=Hidden}
   slot {class=HorizontalBoxSlot, properties {horizontalAlignment=HAlign_Right}}
  }
  TutorialHighlight {widgetClassPath=&TutorialHighlight}
 ]
}
