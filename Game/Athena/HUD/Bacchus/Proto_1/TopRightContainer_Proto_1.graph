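# Prototype Bacchus (mobile) HUD top-right container holding the minimap, respawn indicator, game stats, storm surge and team score.
# Stat widgets and playlist UI extension slots sit below the minimap, with gameplay tags on elements controlling hiding.
# On Construct, mobile touch play moves the respawn indicator into the minimap overlay, bottom-aligned.
# HUD context local player and team win events, bound on Construct and unbound on Destruct, collapse the minimap.
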WidgetBlueprint /Game/Athena/HUD/Bacchus/Proto_1/TopRightContainer_Proto_1
graphs [
 EventGraph {nodes [
   n1 = On Local Team Won() {is=title K2Node_CustomEvent}
   SetVisibility(
    execute=[@n1.then, @n2.then]
    self=Get MinimapContainer() -> (MinimapContainer:Athena Minimap Container Object Reference)
    InVisibility=Collapsed
   )
   n2 = On Local Player Won() {is=title K2Node_CustomEvent}
   Event Construct()
   > Cast To FortPlayerControllerAthena(Object=GetOwningPlayer())
   > SetRespawnMobileLayout()
   > n3 = Sequence()
   n4 = Get AthenaHUDContext!() -> (ReturnValue:Athena HUDContext Object Reference) {is=title K2Node_GetBlueprintContextEx}
   > Bind Event to On Local Player Won(self:Athena HUDContext Object Reference=@n4, Delegate=@n2) {is=title K2Node_AddDelegate}
   Bind Event to On Local Team Won(execute=@n3.then_1, self:Athena HUDContext Object Reference=@n4, Delegate=@n1) {is=title K2Node_AddDelegate}
   Event Destruct()
   > Cast To FortPlayerControllerAthena(Object=GetOwningPlayer())
   > n5 = Sequence()
   > Unbind Event from On Local Player Won(self:Athena HUDContext Object Reference=@n4, Delegate=@n2) {is=title K2Node_RemoveDelegate}
   Unbind Event from On Local Team Won(execute=@n5.then_1, self:Athena HUDContext Object Reference=@n4, Delegate=@n1) {is=title K2Node_RemoveDelegate}
  ]}
 SetRespawnMobileLayout {nodes [
   SetRespawnMobileLayout() {is=title K2Node_FunctionEntry}
   n1 = GetOwningPlayer()
   > Branch(Condition=AND Boolean(A=IsMobileGame(), B=IsUsingTouch(PlayerController=@n1)))
   n2 = Get AthenaRespawnIndicator_0() -> (AthenaRespawnIndicator_0:Athena Respawn Indicator Object Reference)
   > AddChildToOverlay!(
    execute
    self:Overlay Object Reference=Get Overlay_Minimap() -> (Overlay_Minimap:Overlay Object Reference)
    Content:Widget Object Reference=@n2
   ) -> (then, ReturnValue:Overlay Slot Object Reference)
   n3 = SlotAsOverlaySlot(Widget=@n2)
   > SetVerticalAlignment(self=@n3, InVerticalAlignment=VAlign_Bottom)
   > SetHorizontalAlignment(self=@n3, InHorizontalAlignment=HAlign_Fill)
  ]}
]
variables []
widgets {
 info {
  inheritedWidgetCount=3
  namedSlotCount=1
  parentClass=/Script/FortniteUI.BacchusHUDElement
  rootWidgetClass=&ScaleBox
  widgetCount=30
 }
 widgets [
  ScaleBox_HUDWidgetScale {widgetClassPath=&ScaleBox, bInherited=true, widget=ScaleBox_HUDWidgetScale, properties {stretch=UserSpecified}}
  SizeBox_3 {
   widgetClassPath=&SizeBox
   parent=ScaleBox_HUDWidgetScale
   widget=SizeBox_3
   properties {
    bOverride_HeightOverride=true
    bOverride_WidthOverride=true
    heightOverride=340
    maxDesiredHeight=340
    maxDesiredWidth=300
    widthOverride=300
   }
   slot {class=ScaleBoxSlot, properties {horizontalAlignment=HAlign_Fill, verticalAlignment=VAlign_Top}}
  }
  TopRightContainer {widgetClassPath=&VerticalBox, bIsVariable=true, parent=SizeBox_3, widget=TopRightContainer, slot {class=SizeBoxSlot}}
  MapAndGameInfo {widgetClassPath=&VerticalBox, bIsVariable=true, parent=TopRightContainer, widget=MapAndGameInfo, slot {class=VerticalBoxSlot}}
  AthenaRespawnIndicator_0 {
   widgetClassPath=/Game/Athena/HUD/AthenaRespawnIndicator
   bIsVariable=true
   parent=MapAndGameInfo
   widget=AthenaRespawnIndicator_0
   properties {
    visibility=Collapsed
    elementTag {tagName=HUD.Special.RespawnIndicator}
    hUDElementTag {gameplayTags [
      {tagName=Athena.HideForCapture}
      {tagName=HUD.LTM.HideForTryHard}
      {tagName=HUD.Event.HideForSpecialEvent}
      {tagName=Athena.Creative.HideForLoading}
      {tagName=HUD.Hide.WhileLoading}
      {tagName=Athena.LTM.Mash.Title}
      {tagName=HUD.Athena.HideForPostGameSpectate}
      {tagName=HUD.Athena.HideForPostGame}
      {tagName=HUD.LTM.HideForCrucible}
      {tagName=HUD.Athena.HideforRadial}
     ]}
   }
   slot {class=VerticalBoxSlot}
  }
  Overlay_Minimap {
   widgetClassPath=&Overlay
   bIsVariable=true
   parent=MapAndGameInfo
   widget=Overlay_Minimap
   slot {class=VerticalBoxSlot, properties {horizontalAlignment=HAlign_Left, verticalAlignment=VAlign_Center}}
  }
  HorizontalBox_1 {widgetClassPath=&HorizontalBox, parent=Overlay_Minimap, widget=HorizontalBox_1, slot {class=OverlaySlot}}
  MinimapContainer {
   widgetClassPath=/Game/Athena/HUD/AthenaMinimapContainer
   bIsVariable=true
   parent=HorizontalBox_1
   widget=MinimapContainer
   properties {
    hiddenVisibility=Hidden
    hUDElementTag {gameplayTags [
      {tagName=Athena.HideForCapture}
      {tagName=HUD.LTM.HideForTryHard}
      {tagName=HUD.Event.HideForSpecialEvent}
      {tagName=Athena.Creative.HideForLoading}
      {tagName=HUD.Hide.WhileLoading}
      {tagName=HUD.LTM.HideForCrucible}
      {tagName=Athena.Creative.HUDController.HideMiniMap}
     ]}
   }
   slot {class=HorizontalBoxSlot}
  }
  SizeBox_0 {
   widgetClassPath=&SizeBox
   parent=HorizontalBox_1
   widget=SizeBox_0
   properties {bOverride_WidthOverride=true, widthOverride=33}
   slot {class=HorizontalBoxSlot}
  }
  SizeBox_1 {
   widgetClassPath=&SizeBox
   parent=Overlay_Minimap
   widget=SizeBox_1
   properties {bOverride_HeightOverride=true, bOverride_WidthOverride=true, heightOverride=300, widthOverride=300}
   slot {class=OverlaySlot, properties {verticalAlignment=VAlign_Bottom}}
  }
  UIExtensionSlot_OverMiniMap {
   widgetClassPath=/Script/FortniteUI.FortPlaylistUIExtensionSlot
   bIsVariable=true
   parent=SizeBox_1
   widget=UIExtensionSlot_OverMiniMap
   properties {entryBoxType=Vertical, entryWidgetClass=&CommonUserWidget, entrySizeRule {sizeRule=Fill, value=1}}
   slot {class=SizeBoxSlot}
  }
  FortHUDElementWrapper {
   widgetClassPath=/Game/Athena/HUD/FortHUDElementWrapper
   bIsVariable=true
   parent=MapAndGameInfo
   widget=FortHUDElementWrapper
   properties {elementTag {tagName=HUD.BelowMinimap.GameStatus}}
   slot {class=VerticalBoxSlot, properties {horizontalAlignment=HAlign_Left, verticalAlignment=VAlign_Top}}
  }
  SizeBox_2 {
   widgetClassPath=&SizeBox
   namedSlotHost=FortHUDElementWrapper
   widget=SizeBox_2
   properties {bOverride_MaxDesiredWidth=true, maxDesiredWidth=400, minDesiredWidth=300, widthOverride=300}
  }
  VerticalBox_0 {
   widgetClassPath=&VerticalBox
   parent=SizeBox_2
   widget=VerticalBox_0
   slot {class=SizeBoxSlot, properties {verticalAlignment=VAlign_Center}}
  }
  HorizontalBox_0 {
   widgetClassPath=&HorizontalBox
   parent=VerticalBox_0
   widget=HorizontalBox_0
   properties {renderTransform {angle=0, scale=<0.8 0.8>, shear=<0 0>, translation=<0 1>}, renderTransformPivot=<0 0>}
   slot {class=VerticalBoxSlot, properties {horizontalAlignment=HAlign_Left, verticalAlignment=VAlign_Top}}
  }
  AthenaGamePhaseWidget {
   widgetClassPath=/Game/Athena/HUD/Phase/AthenaGamePhaseWidget
   bIsVariable=true
   parent=HorizontalBox_0
   widget=AthenaGamePhaseWidget
   properties {
    elementTag {tagName=MobileHUD.Stats.GamePhase}
    hUDElementTag {gameplayTags [
      {tagName=Athena.HideForCapture}
      {tagName=HUD.Athena.Hide.WithInventory}
      {tagName=HUD.Athena.Hide.WithMap}
      {tagName=HUD.LTM.HideForTryHard}
      {tagName=HUD.Event.HideForSpecialEvent}
      {tagName=Athena.Creative.HideForLoading}
      {tagName=HUD.Hide.WhileLoading}
      {tagName=HUD.LTM.HideForCrucible}
      {tagName=Athena.Creative.HUDController.HideStormTimer}
      {tagName=Athena.LTM.Mash.Title}
      {tagName=HUD.Athena.HideForPostGame}
      {tagName=HUD.Athena.HideforRadial}
      {tagName=HUD.Hide.CreativeJoiningLobby}
      {tagName=HUD.Athena.GamePhase.MinimapInfo}
     ]}
   }
   slot {class=HorizontalBoxSlot, properties {horizontalAlignment=HAlign_Center, padding=<ltrb 1 0 1 0>}}
  }
  AthenaPlayersLeftWidget {
   widgetClassPath=/Game/Athena/HUD/AthenaPlayersLeftWidget
   bIsVariable=true
   parent=HorizontalBox_0
   widget=AthenaPlayersLeftWidget
   properties {
    elementTag {tagName=MobileHUD.Stats.PlayersLeft}
    hUDElementTag {gameplayTags [
      {tagName=Athena.HideForCapture}
      {tagName=HUD.Athena.Hide.WithInventory}
      {tagName=HUD.Athena.Hide.WithMap}
      {tagName=HUD.LTM.HideForTryHard}
      {tagName=HUD.Event.HideForSpecialEvent}
      {tagName=Athena.Creative.HideForLoading}
      {tagName=HUD.Hide.WhileLoading}
      {tagName=HUD.LTM.HideForCrucible}
      {tagName=Athena.Creative.HUDController.HidePlayerCount}
      {tagName=HUD.Athena.HideForPostGame}
      {tagName=HUD.Athena.HideforRadial}
      {tagName=HUD.Hide.CreativeJoiningLobby}
      {tagName=Athena.LTM.Mash.Title}
      {tagName=HUD.Athena.PlayersLeft}
     ]}
   }
   slot {class=HorizontalBoxSlot, properties {horizontalAlignment=HAlign_Center, padding=<ltrb 1 0 1 0>}}
  }
  AthenaKillsWidget {
   widgetClassPath=/Game/Athena/HUD/AthenaKillsWidget
   bIsVariable=true
   parent=HorizontalBox_0
   widget=AthenaKillsWidget
   properties {
    elementTag {tagName=MobileHUD.Stats.Kills}
    hUDElementTag {gameplayTags [
      {tagName=Athena.HideForCapture}
      {tagName=HUD.Athena.Hide.WithInventory}
      {tagName=HUD.Athena.Hide.WithMap}
      {tagName=HUD.LTM.HideForTryHard}
      {tagName=HUD.Event.HideForSpecialEvent}
      {tagName=Athena.Creative.HideForLoading}
      {tagName=HUD.Hide.WhileLoading}
      {tagName=HUD.LTM.HideForCrucible}
      {tagName=Athena.Creative.HUDController.HideEliminationCounter}
      {tagName=Athena.LTM.Mash.Title}
      {tagName=HUD.Athena.HideForPostGame}
      {tagName=HUD.Athena.HideforRadial}
      {tagName=HUD.Athena.HideOnSpectate}
      {tagName=HUD.Hide.CreativeJoiningLobby}
      {tagName=HUD.Athena.Kills}
     ]}
   }
   slot {class=HorizontalBoxSlot, properties {horizontalAlignment=HAlign_Center, padding=<ltrb 1 0 1 0>}}
  }
  AthenaTeamKillsWidget {
   widgetClassPath=/Game/Athena/HUD/AthenaTeamKillsWidget
   bIsVariable=true
   parent=HorizontalBox_0
   widget=AthenaTeamKillsWidget
   properties {
    visibility=Collapsed
    elementTag {tagName=MobileHUD.Stats.TeamKills}
    hUDElementTag {gameplayTags [
      {tagName=HUD.Athena.Hide.WithInventory}
      {tagName=HUD.Athena.Hide.WithMap}
      {tagName=HUD.LTM.HideForTryHard}
      {tagName=HUD.Event.HideForSpecialEvent}
      {tagName=HUD.Hide.WhileLoading}
      {tagName=HUD.LTM.HideForCrucible}
      {tagName=HUD.Athena.HideForPostGame}
      {tagName=HUD.Athena.HideforRadial}
      {tagName=HUD.Hide.CreativeJoiningLobby}
      {tagName=Athena.Creative.HideForLoading}
      {tagName=Athena.Creative.HUDController.HideTeamInfo}
      {tagName=Athena.HideForCapture}
      {tagName=Athena.LTM.Mash.Title}
      {tagName=HUD.Athena.TeamKills}
     ]}
   }
   slot {class=HorizontalBoxSlot, properties {horizontalAlignment=HAlign_Center, padding=<ltrb 1 0 1 0>}}
  }
  GameInfoBox_Bacchus {
   widgetClassPath=&FortPlaylistUIExtensionSlot
   bIsVariable=true
   parent=HorizontalBox_0
   widget=GameInfoBox_Bacchus
   properties {
    entryHorizontalAlignment=HAlign_Left
    entryVerticalAlignment=VAlign_Top
    entryWidgetClass=&CommonUserWidget
    numDesignerPreviewEntries=1
    extensionSlots [GameInfoBelowMap, GameInfoBox]
   }
   slot {class=HorizontalBoxSlot, properties {horizontalAlignment=HAlign_Center, padding=<ltrb 1 0 0 0>}}
  }
  HorizontalBox_2 {
   widgetClassPath=&HorizontalBox
   parent=VerticalBox_0
   widget=HorizontalBox_2
   slot {class=VerticalBoxSlot, properties {padding=<ltrb 0 5 0 5>}}
  }
  SizeBox_4 {
   widgetClassPath=&SizeBox
   parent=HorizontalBox_2
   widget=SizeBox_4
   slot {class=HorizontalBoxSlot, properties {size {sizeRule=Fill, value=1}}}
  }
  ExtensionSlotUnderSquadInfo {
   widgetClassPath=&FortPlaylistUIExtensionSlot
   bIsVariable=true
   parent=SizeBox_4
   widget=ExtensionSlotUnderSquadInfo
   properties {
    entryBoxType=Vertical
    entryWidgetClass=&CommonUserWidget
    maxElementSize=1
    entrySpacing=<0 3>
    extensionSlots [UnderSquadInfo, MobileHUDUnderMinimap]
   }
   slot {class=SizeBoxSlot}
  }
  VerticalBox_71 {widgetClassPath=&VerticalBox, parent=VerticalBox_0, widget=VerticalBox_71, slot {class=VerticalBoxSlot}}
  SizeBox_107 {widgetClassPath=&SizeBox, parent=VerticalBox_71, widget=SizeBox_107, slot {class=VerticalBoxSlot}}
  UIExtensionSlot_UnderMapMessagesContainer {
   widgetClassPath=&FortPlaylistUIExtensionSlot
   bIsVariable=true
   parent=SizeBox_107
   widget=UIExtensionSlot_UnderMapMessagesContainer
   properties {entryBoxType=Vertical, entryWidgetClass=&CommonUserWidget, extensionSlots [UnderMinimapMessageContainer]}
   slot {class=SizeBoxSlot, properties {horizontalAlignment=HAlign_Left}}
  }
  AthenaStormSurgeWidget {
   widgetClassPath=/Game/Athena/HUD/AthenaStormSurgeWidget
   bIsVariable=true
   parent=TopRightContainer
   widget=AthenaStormSurgeWidget
   properties {isBacchus=true, visibility=Collapsed}
   slot {class=VerticalBoxSlot, properties {horizontalAlignment=HAlign_Left, padding=<ltrb 18 0 0 0>}}
  }
  TeamScoreWidget {
   widgetClassPath=/Game/Athena/HUD/AthenaTeamScoreWidget
   bIsVariable=true
   parent=TopRightContainer
   widget=TeamScoreWidget
   properties {
    isBacchus=true
    visibility=Collapsed
    hUDElementTag {gameplayTags [
      {tagName=HUD.Athena.Hide.WithMap}
      {tagName=HUD.LTM.HideForTryHard}
      {tagName=HUD.Event.HideForSpecialEvent}
      {tagName=HUD.Hide.WhileLoading}
      {tagName=Athena.HideForCapture}
     ]}
    renderTransformPivot=<1 0.5>
   }
   slot {class=VerticalBoxSlot, properties {horizontalAlignment=HAlign_Left, verticalAlignment=VAlign_Bottom}}
  }
  Image_Background {widgetClassPath=&Image, bInherited=true, widget=null}
  TutorialHighlight {widgetClassPath=&TutorialHighlight}
 ]
}
